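; PCG editor-mode area tool that scatters weighted static meshes or actors inside spline outlines.
; It samples a spline-derived surface, using distance to the sampled spline for edge-falloff culling and scaling, then thins by probability or fixed count.
; Points receive weighted assets, global and random transforms, optional overlap pruning, and optional snapping or alignment via ray-hit projection cast from combined bounds.
; With no assets supplied, tagged points render as debug previews instead of spawning.
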(graph AreaTool
  (params (PointsPerSquaredMeter :type Double :default 0.1) (PointSize :type Vector2 :default (1e+02 1e+02)) (Uniformity :type Double) (SamplingSeed :type Int :default 492183294) (ProbabilityFilter :type Double :default 1e+02) (UseFixedCountFilter :type Bool) (FixedCountFilter :type Int :default 10) (FilterSeed :type Int :default -193199241) (EdgeFallofSize :type Double :default 1e+03) (EdgeFalloffPower :type Double :default 1.0) (EdgeFalloffFilter :type Double) (RemoveOverlapingPoints :type Bool) (UseMeshBounds :type Bool) (StaticMeshes :type (Array (Struct /PCG/EdMode/Resources/PCGEdMode_WeightedStaticMeshReference.PCGEdMode_WeightedStaticMeshReference))) (SpawnActors :type Bool) (ActorClasses :type (Array (Struct /PCG/EdMode/Resources/PCGEdMode_WeightedClassReference.PCGEdMode_WeightedClassReference))) (GlobalPositionOffset :type Vector) (GlobalRotation :type Rotator) (AbsoluteGlobalRotation :type Bool) (GlobalScaleFactor :type Vector :default (1.0 1.0 1.0)) (EdgeFalloffScaleMin :type Double :default 1.0) (RandomPositionOffsetMin :type Vector) (RandomPositionOffsetMax :type Vector) (RandomRotationMin :type Rotator) (RandomRotationMax :type Rotator) (RandomScaleFactorMin :type Vector :default (1.0 1.0 1.0)) (RandomScaleFactorMax :type Vector :default (1.0 1.0 1.0)) (UniformRandomScaleFactor :type Bool :default true) (RandomSeed :type Int :default 2118139241) (SnapToSurface :type Bool) (AlignToSurface :type Bool) (SnapToSurfaceMaxDistance :type Double :default 1e+03))
  (bind GetSplineData_0 (Get%20Spline%20Data))
  (bind Select_1 (Select :bUseInputB-param SpawnActors :Input%20A StaticMeshes :Input%20B ActorClasses))
  (bind RemoveEmptyData_1 (Remove%20Empty%20Data))
  (bind DataNum_1 (Data%20Count))
  (bind AttributeMathsOp_6 (Attribute%20Maths%20Op :Operation Divide :InA ProbabilityFilter))
  (bind AttributeMathsOp_0 (Attribute%20Maths%20Op :Operation OneMinus :In Uniformity))
  (bind AttributeMathsOp_4 (Attribute%20Maths%20Op :Operation Divide :InA PointSize))
  (bind AttributeMathsOp_3 (Attribute%20Maths%20Op :Operation Max))
  (bind FilterDataByType_0 (Filter%20Data%20By%20Type :TargetType (struct :Ids [(struct :Struct /Script/PCG.PCGDataTypeInfoSpline)])))
  (bind CreateSurfaceFromSpline_0 (Create%20Surface%20From%20Spline))
  (bind SurfaceSampler_0 (Surface%20Sampler :bApplyDensityToPoints false :bUnbounded true :Looseness 0.0 :PointsPerSquaredMeter-param PointsPerSquaredMeter :Seed-param SamplingSeed))
  (bind RemoveEmptyData_0 (Remove%20Empty%20Data))
  (bind SplineToSegment_0 (Spline%20to%20Segment))
  (bind AttributeReduce_0 (Attribute%20Reduce :InputSource $Extents.x :Operation Sum))
  (bind AttributeMathsOp_5 (Attribute%20Maths%20Op :Operation Multiply))
  (bind AttributeMathsOp_8 (Attribute%20Maths%20Op :Operation Divide))
  (bind AttributeMathsOp_12 (Attribute%20Maths%20Op :Operation Max))
  (bind AttributeMathsOp_11 (Attribute%20Maths%20Op :Operation Min))
  (bind SplineSampler_0 (Spline%20Sampler :SamplerParams (struct :Mode Distance :DistanceIncrement 25.0 :bFitToCurve true :InteriorOrientation FollowCurvature :InteriorDensityFalloffCurve (struct :EditorCurveData (struct :Keys [(struct :Value 1.0) (struct :Time 1.0)])) :bUnbounded true :PointSteepness 1.0)))
  (bind Distance_0 (Distance :bOutputToAttribute false :bSetDensity true :MaximumDistance 2e+03 :SourceShape Center :TargetShape Center :MaximumDistance-param EdgeFallofSize))
  (bind AttributeMathsOp_10 (Attribute%20Maths%20Op :InputSource1 $Density :Operation Pow :InB EdgeFalloffPower))
  (bind AttributeFilter_1 (Filter%20Attribute%20Elements :bGenerateOutputDataEvenIfEmpty false :ThresholdAttribute @Last :Filter EdgeFalloffFilter))
  (bind AttributeMathsOp_9 (Attribute%20Maths%20Op :InputSource1 $Density :Operation Max :InB EdgeFalloffScaleMin))
  (bind AttributeMathsOp_7 (Attribute%20Maths%20Op :InputSource1 $Scale :InputSource2 $Density :Operation Multiply))
  (bind Branch_4 (Branch :bOutputToB-param UseFixedCountFilter))
  (bind AttributeNoise_0 (Attribute%20Noise :Seed-param FilterSeed))
  (bind AttributeFilter_0 (Filter%20Attribute%20Elements :bGenerateOutputDataEvenIfEmpty false :Operator Lesser :ThresholdAttribute @Last))
  (bind RandomChoice_0 (Random%20Choice :FixedNumber 10 :FixedNumber-param FixedCountFilter :Seed-param FilterSeed))
  (bind Branch_0 (Branch))
  (bind MatchAndSetAttributes_1 (Match%20And%20Set%20Attributes :bUseWeightAttribute true :WeightAttribute Weight))
  (bind AddTags_0 (Add%20Tags :TagsToAdd "ShowPreviewMesh"))
  (bind Branch_2 (Branch :bOutputToB-param SpawnActors))
  (bind Branch_5 (Branch :bOutputToB-param UseMeshBounds))
  (bind BoundsFromMesh_1 (Bounds%20From%20Mesh :MeshAttribute StaticMesh))
  (bind TransformPoints_1 (Transform%20Points :bUniformScale false :bAbsoluteRotation-param AbsoluteGlobalRotation :OffsetMax-param GlobalPositionOffset :OffsetMin-param GlobalPositionOffset :RotationMax-param GlobalRotation :RotationMin-param GlobalRotation :ScaleMax-param GlobalScaleFactor :ScaleMin-param GlobalScaleFactor))
  (bind TransformPoints_2 (Transform%20Points :bUniformScale-param UniformRandomScaleFactor :OffsetMax-param RandomPositionOffsetMax :OffsetMin-param RandomPositionOffsetMin :RotationMax-param RandomRotationMax :RotationMin-param RandomRotationMin :ScaleMax-param RandomScaleFactorMax :ScaleMin-param RandomScaleFactorMin :Seed-param RandomSeed))
  (bind Branch_6 (Branch :bOutputToB-param RemoveOverlapingPoints))
  (bind SelfPruning_0 (Self%20Pruning :Parameters (struct :ComparisonSource $Extents :bRandomizedPruning false)))
  (bind CombinePoints_0 (Combine%20Points :bUseFirstPointTransform false))
  (bind AttributeMathsOp_2 (Attribute%20Maths%20Op :InputSource1 $Position.z :InputSource2 $BoundsMax.Z))
  (bind ExtractAttribute_0 (Extract%20Attribute%20at%20Index :bForceOutputAttributeToBeInElementsDomain true :InputSource $Position))
  (bind WorldRayHitQuery_1 (World%20Ray%20Hit%20Query :QueryParams (struct :bOverrideDefaultParams true :RayLength 1e+03) :RayLength SnapToSurfaceMaxDistance))
  (bind Projection_1 (Projection :bKeepZeroDensityPoints true :ProjectionParams (struct :AttributeMode IncludeAttributes :AttributeMergeOperation SourceValue) :bProjectPositions SnapToSurface :bProjectRotations AlignToSurface))
  (bind FilterDataByTag_0 (Filter%20Data%20By%20Tag :SelectedTags "ShowPreviewMesh"))
  (bind Branch_1 (Branch :bOutputToB-param SpawnActors))
  (bind StaticMeshSpawner_1 (Static%20Mesh%20Spawner :MeshSelectorParameters (class:/Script/PCG.PCGMeshSelectorByAttribute :AttributeName StaticMesh :TemplateDescriptor (struct :bUseDefaultCollision true)) :MeshSelectorType /Script/PCG.PCGMeshSelectorByAttribute))
  (bind SpawnActor_1 (Spawn%20Actor :bSpawnByAttribute true :PostSpawnFunctionNames [None] :SpawnAttribute Class))
  (bind Debug_0 (Debug))
  (connect Branch_0 Output%20A AddTags_0 In)
  (connect AttributeMathsOp_6 Out AttributeFilter_0 Filter)
  (connect AttributeNoise_0 Out AttributeFilter_0 In)
  (connect AttributeMathsOp_10 Out AttributeFilter_1 In)
  (connect CombinePoints_0 Out AttributeMathsOp_2 InA)
  (connect CombinePoints_0 Out AttributeMathsOp_2 InB)
  (connect AttributeMathsOp_4 Out AttributeMathsOp_3 InA)
  (connect AttributeReduce_0 Out AttributeMathsOp_5 InA)
  (connect AttributeFilter_1 InsideFilter AttributeMathsOp_7 InA)
  (connect AttributeMathsOp_9 Out AttributeMathsOp_7 InB)
  (connect AttributeMathsOp_5 Out AttributeMathsOp_8 InA)
  (connect AttributeFilter_1 InsideFilter AttributeMathsOp_9 InA)
  (connect Distance_0 Out AttributeMathsOp_10 InA)
  (connect AttributeMathsOp_12 Out AttributeMathsOp_11 InA)
  (connect AttributeMathsOp_8 Out AttributeMathsOp_12 InA)
  (connect Branch_4 Output%20A AttributeNoise_0 In)
  (connect SplineToSegment_0 Out AttributeReduce_0 In)
  (connect Branch_5 Output%20B BoundsFromMesh_1 In)
  (connect DataNum_1 Out Branch_0 bOutputToB)
  (connect AttributeFilter_0 InsideFilter Branch_0 In)
  (connect RandomChoice_0 Chosen Branch_0 In)
  (connect FilterDataByTag_0 OutsideFilter Branch_1 In)
  (connect MatchAndSetAttributes_1 Out Branch_2 In)
  (connect AttributeMathsOp_7 Out Branch_4 In)
  (connect Branch_2 Output%20A Branch_5 In)
  (connect TransformPoints_2 Out Branch_6 In)
  (connect Branch_6 Output%20A CombinePoints_0 In)
  (connect SelfPruning_0 Out CombinePoints_0 In)
  (connect FilterDataByType_0 InsideFilter CreateSurfaceFromSpline_0 In)
  (connect RemoveEmptyData_1 Out DataNum_1 In)
  (connect FilterDataByTag_0 InsideFilter Debug_0 In)
  (connect RemoveEmptyData_0 Out Distance_0 Source)
  (connect SplineSampler_0 Out Distance_0 Target)
  (connect AttributeMathsOp_2 Out ExtractAttribute_0 In)
  (connect Projection_1 Out FilterDataByTag_0 In)
  (connect GetSplineData_0 Out FilterDataByType_0 In)
  (connect Branch_0 Output%20B MatchAndSetAttributes_1 In)
  (connect Select_1 Out MatchAndSetAttributes_1 Match%20Data)
  (connect Branch_6 Output%20A Projection_1 In)
  (connect SelfPruning_0 Out Projection_1 In)
  (connect WorldRayHitQuery_1 Out Projection_1 Projection%20Target)
  (connect Branch_4 Output%20B RandomChoice_0 In)
  (connect SurfaceSampler_0 Out RemoveEmptyData_0 In)
  (connect Select_1 Out RemoveEmptyData_1 In)
  (connect Branch_6 Output%20B SelfPruning_0 In)
  (connect Branch_1 Output%20B SpawnActor_1 In)
  (connect AttributeMathsOp_11 Out SplineSampler_0 DistanceIncrement)
  (connect FilterDataByType_0 InsideFilter SplineSampler_0 Spline)
  (connect FilterDataByType_0 InsideFilter SplineToSegment_0 In)
  (connect Branch_1 Output%20A StaticMeshSpawner_1 In)
  (connect AttributeMathsOp_0 Out SurfaceSampler_0 Looseness)
  (connect AttributeMathsOp_3 Out SurfaceSampler_0 PointExtents)
  (connect CreateSurfaceFromSpline_0 Out SurfaceSampler_0 Surface)
  (connect AddTags_0 Out TransformPoints_1 In)
  (connect BoundsFromMesh_1 Out TransformPoints_1 In)
  (connect Branch_2 Output%20B TransformPoints_1 In)
  (connect Branch_5 Output%20A TransformPoints_1 In)
  (connect TransformPoints_1 Out TransformPoints_2 In)
  (connect ExtractAttribute_0 Out WorldRayHitQuery_1 RayOrigin))
